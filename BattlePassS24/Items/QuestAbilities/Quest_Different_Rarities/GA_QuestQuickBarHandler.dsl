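; Season 24 Battle Pass quest ability that reports quick bar item info to the Athena quest manager, apparently for a different-rarities quest.
; On the server send-quest event, it gets the owning player controller's gameplay tags and adds the custom verb tag.
; It sends those tags with all item tags as a complex custom stat event with a count of one.
; The verb and item tags come from FortAthenaQuestGA_QuickBarInfo.
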(event EventServerSendQuestEvent
  ;[1 1 1] Send Quest event
  (bind _asfort_player_controller_athena (Utilities|Casting|CastToFortPlayerControllerAthena (Fortnite|GetControllerfromActor (Ability|GetOwningActorfromActorInfo))))
  (bind _returnvalue (GameplayTags|GetOwnedGameplayTags _asfort_player_controller_athena :!class /Script/GameplayTags.GameplayTagAssetInterface))
  (GameplayTags|AddGameplayTag _returnvalue (Variables|FortAthenaQuestGA_QuickBarInfo|GetCustomVerbNameTag) :!class /Script/GameplayTags.BlueprintGameplayTagLibrary)
  (Quest|SendComplexCustomStatEvent (Quests|GetQuestManager _asfort_player_controller_athena "Athena") 0 _returnvalue (Variables|FortAthenaQuestGA_QuickBarInfo|GetAllItemTags) 1))
